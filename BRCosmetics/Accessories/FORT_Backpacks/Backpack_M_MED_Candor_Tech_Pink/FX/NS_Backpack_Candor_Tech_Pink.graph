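# Niagara system for the pink Candor Tech backpack cosmetic, placing five stickers (01, 03–06) as persistent burst-spawned mesh particles.
# Each two-mesh sticker emitter pairs with a RainbowLine overlay, positioned and sized by User parameters; Sticker06 reuses Sticker05 Size.
# System brightness and time-of-day values scale sticker color.
# User.IsFrom3D switches rainbow flatness and chooses fixed or time-of-day strength; the Glow sprite emitter is disabled.
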NiagaraSystem /BRCosmetics/Accessories/FORT_Backpacks/Backpack_M_MED_Candor_Tech_Pink/FX/NS_Backpack_Candor_Tech_Pink
dependencies {
 usedDataInterfaces [&NiagaraDataInterfaceCurve]
 usedDynamicInputs [&RandomRangeFloat, &FN_FrontEndFloatModifier]
 usedModules [
  &self:SystemSpawnScript.SystemScriptSource.SystemScriptGraph.NiagaraNodeAssignment_0.SetVariables_#1
  &self:SystemSpawnScript.SystemScriptSource.SystemScriptGraph.NiagaraNodeAssignment_2.SetVariables_#2
  &SystemState
  &self:SystemSpawnScript.SystemScriptSource.SystemScriptGraph.NiagaraNodeAssignment_1.SetVariables_#3
  &self:SystemSpawnScript.SystemScriptSource.SystemScriptGraph.NiagaraNodeAssignment_3.SetVariables_#4
  &self:SystemSpawnScript.SystemScriptSource.SystemScriptGraph.NiagaraNodeAssignment_4.SetVariables_#5
 ]
}
emitters {
 Glow {
  bEnabled=false
  emitterUpdateScript [
   &EmitterState(
    Inactive Response=::NewEnumerator1
    Life Cycle Mode=::NewEnumerator1
    Loop Behavior=::NewEnumerator1
    Loop Duration Mode=::NewEnumerator1
    UseLoopDelay=false
   )
   &SpawnBurst_Instantaneous(Use Spawn Probability=false, UseLoopCountLimit=false)
  ]
  particleSpawnScript [&InitializeParticle(
    Color=<rgba 0 0.0807 0.1615 1>
    Position Offset=<0 -13 -3>
    Sprite Size Mode=::NewEnumerator3
    Uniform Sprite Size=50
    UsePositionOffset=true
    Write Material Random=true
   )]
  particleUpdateScript [
   &ParticleState(Kill Particles When Lifetime Has Elapsed=false, Let Infinitely Lived Particles Die When Emitter Deactivates=true)
   &ScaleColor(
    Scale Alpha=&FNTimeOfDayScalar(Night Scale=0.5, Morning Scale=0.5, Evening Scale=0.5)
    ScaleA=true
    ScaleRGB=true
   )
   &CameraOffset(Camera Offset Amount=20)
  ]
  rendererClasses [&NiagaraSpriteRendererProperties]
  renderers [{rendererClass=&NiagaraSpriteRendererProperties}]
 }
 Sticker01 {
  emitterUpdateScript [
   &EmitterState(
    Inactive Response=::NewEnumerator1
    Life Cycle Mode=::NewEnumerator1
    Loop Behavior=::NewEnumerator1
    Loop Duration Mode=::NewEnumerator1
    UseLoopDelay=false
   )
   &SpawnBurst_Instantaneous(Use Spawn Probability=false, UseLoopCountLimit=false)
  ]
  particleSpawnScript [
   &InitializeParticle(
    Color=$System.Base color
    Mesh Scale=&Multiply_VectorByFloat(Vector=<1 1 1>, Float=0.5)
    Mesh Scale Mode=::NewEnumerator4
    Position Offset=&Add_Vector(A=Linked() {linkedVariable {name=User.Sticker01 Position, type=NiagaraPosition}}, B=<0 0 0>)
    UsePositionOffset=true
    Write Material Random=true
   )
   &DynamicMaterialParameters(
    Index 0 Param 1=$System.Random
    Index 0 Param 2=&MakeFloatFromVector(Vector=$Particles.Scale, Channel=::NewEnumerator0)
    Module.Param0WriteEnabled=true
    Module.Param1WriteEnabled=true
    Module.Param2WriteEnabled=false
    Module.Param3WriteEnabled=false
   )
   &InitialMeshOrientation(Mesh Orientation Mode=::NewEnumerator2, Rotation=<-0.05 0 0>, Use Rotation Vector=true)
   /SetVariables(Particles.Mesh02 size:Vector3f=&Multiply_VectorByFloat(Vector=<0.9 2 1>, Float=$User.Sticker01 Size)) {isSetParameters=true, module=SetVariables_#6}
   /SetVariables(Particles.Mesh02 Color:LinearColor=<rgba 1 1 1 3>) {isSetParameters=true, module=SetVariables_#7}
   /SetVariables(Particles.Mesh02 Position:NiagaraPosition=&Add_Vector(A=Linked() {linkedVariable {name=Particles.Position, type=NiagaraPosition}}, B=<0 7 -2.5>)) {isSetParameters=true, module=SetVariables_#8}
  ]
  particleUpdateScript [
   &ParticleState(Kill Particles When Lifetime Has Elapsed=false, Let Infinitely Lived Particles Die When Emitter Deactivates=true)
   &ScaleMeshSize(Scale Factor=&Multiply_VectorByFloat(Vector=<0.9 1 1>, Float=$User.Sticker01 Size))
   &ScaleColor(
    Scale RGB=&Multiply_VectorByFloat(Vector=&Multiply_VectorByFloat(Vector=<1 1 1>, Float=$System.Bright), Float=$System.TOD)
    ScaleA=true
    ScaleRGB=true
   )
  ]
  rendererClasses [&NiagaraMeshRendererProperties]
  renderers [
   {rendererClass=&NiagaraMeshRendererProperties, rendererIndex=0}
   {rendererClass=&NiagaraMeshRendererProperties, rendererIndex=1}
  ]
 }
 Sticker01_PresetFX_RainbowLine {
  emitterUpdateScript [
   &EmitterState(
    Inactive Response=::NewEnumerator1
    Life Cycle Mode=::NewEnumerator1
    Loop Behavior=::NewEnumerator1
    Loop Duration Mode=::NewEnumerator1
    UseLoopDelay=false
   )
   &SpawnBurst_Instantaneous(Use Spawn Probability=false, UseLoopCountLimit=false)
   /SetVariables(Emitter.RainbowLine_Flat:NiagaraFloat=&MakeCustomFloatFromBool(Bool=&SetBoolByFloatComparison(A=$User.IsFrom3D, B=0), True Float=0, False Float=-1)) {isSetParameters=true, module=SetVariables_#9}
   /SetVariables(Emitter.RainbowLine_Strength:NiagaraFloat=&MakeCustomFloatFromBool(
     Bool=&SetBoolByFloatComparison(A=$User.IsFrom3D, B=0)
     True Float=4
     False Float=&FNTimeOfDayScalar(Night Scale=1.5, Morning Scale=0.4, Evening Scale=0.3)
    )) {isSetParameters=true, module=SetVariables_#10}
   /SetVariables(Emitter.RainbowLine_Power=0.5) {isSetParameters=true, module=SetVariables_#11}
   /SetVariables(Emitter.RainbowLine_Density=5) {isSetParameters=true, module=SetVariables_#12}
   /SetVariables(Emitter.RainbowLine_FlowSpeed=0.15) {isSetParameters=true, module=SetVariables_#13}
  ]
  particleSpawnScript [
   &InitializeParticle(
    Mesh Scale Mode=::NewEnumerator3
    Mesh Uniform Scale=0.5
    Position Offset=&Add_Vector(A=<0 1 -0.5>, B=Linked() {linkedVariable {name=User.Sticker01 Position, type=NiagaraPosition}})
    Sprite Size Mode=::NewEnumerator3
    Uniform Sprite Size=70
    UsePositionOffset=true
    Write Material Random=true
   )
   &DynamicMaterialParameters(
    Index 0 Param 1=$System.Random
    Module.Param0WriteEnabled=true
    Module.Param1WriteEnabled=false
    Module.Param2WriteEnabled=false
    Module.Param3WriteEnabled=false
   )
   /SetVariables(Particles.RainbowLineSize:Vector2f=&Multiply_Vector2D_ByFloat(Float=0.5, Vector2D=$Particles.SpriteSize)) {isSetParameters=true, module=SetVariables_#14}
   &InitialMeshOrientation(Mesh Orientation Mode=::NewEnumerator2, Rotation=<-0.05 0 0>, Use Rotation Vector=true)
  ]
  particleUpdateScript [
   &ParticleState(Kill Particles When Lifetime Has Elapsed=false, Let Infinitely Lived Particles Die When Emitter Deactivates=true)
   &ScaleMeshSize(Scale Factor=&Multiply_VectorByFloat(Vector=<0.9 1 1>, Float=$User.Sticker01 Size))
  ]
  rendererClasses [&NiagaraMeshRendererProperties]
  renderers [{rendererClass=&NiagaraMeshRendererProperties, rendererIndex=0}]
 }
 Sticker03 {
  emitterUpdateScript [
   &EmitterState(
    Inactive Response=::NewEnumerator1
    Life Cycle Mode=::NewEnumerator1
    Loop Behavior=::NewEnumerator1
    Loop Duration Mode=::NewEnumerator1
    UseLoopDelay=false
   )
   &SpawnBurst_Instantaneous(Use Spawn Probability=false, UseLoopCountLimit=false)
  ]
  particleSpawnScript [
   &InitializeParticle(
    Color=$System.Base color
    Mesh Scale=&Multiply_VectorByFloat(Vector=<1 1 1>, Float=0.5)
    Mesh Scale Mode=::NewEnumerator4
    Position Offset=&Add_Vector(A=Linked() {linkedVariable {name=User.Sticker03 Position, type=NiagaraPosition}}, B=<0 0 0>)
    UsePositionOffset=true
    Write Material Random=true
   )
   &DynamicMaterialParameters(
    Index 0 Param 1=0.5
    Index 0 Param 2=&MakeFloatFromVector(Vector=$Particles.Scale, Channel=::NewEnumerator0)
    Module.Param0WriteEnabled=true
    Module.Param1WriteEnabled=true
    Module.Param2WriteEnabled=false
    Module.Param3WriteEnabled=false
   )
   &InitialMeshOrientation(Mesh Orientation Mode=::NewEnumerator2, Rotation=<0.04 0 0>, Use Rotation Vector=true)
   /SetVariables(Particles.Mesh02 size:Vector3f=&Multiply_VectorByFloat(Vector=<1 2.5 1>, Float=$User.Sticker03 Size)) {isSetParameters=true, module=SetVariables_#6}
   /SetVariables(Particles.Mesh02 Color:LinearColor=<rgba 1 1 1 3>) {isSetParameters=true, module=SetVariables_#7}
   /SetVariables(Particles.Mesh02 Position:NiagaraPosition=&Add_Vector(A=Linked() {linkedVariable {name=Particles.Position, type=NiagaraPosition}}, B=<0 7 0.8>)) {isSetParameters=true, module=SetVariables_#8}
  ]
  particleUpdateScript [
   &ParticleState(Kill Particles When Lifetime Has Elapsed=false, Let Infinitely Lived Particles Die When Emitter Deactivates=true)
   &ScaleMeshSize(Scale Factor=&Multiply_VectorByFloat(Vector=<1 1 1>, Float=$User.Sticker03 Size))
   &ScaleColor(
    Scale RGB=&Multiply_VectorByFloat(Vector=&Multiply_VectorByFloat(Vector=<1 1 1>, Float=$System.Bright), Float=$System.TOD)
    ScaleA=true
    ScaleRGB=true
   )
  ]
  rendererClasses [&NiagaraMeshRendererProperties]
  renderers [
   {rendererClass=&NiagaraMeshRendererProperties, rendererIndex=0}
   {rendererClass=&NiagaraMeshRendererProperties, rendererIndex=1}
  ]
 }
 Sticker03_PresetFX_RainbowLine002 {
  emitterUpdateScript [
   &EmitterState(
    Inactive Response=::NewEnumerator1
    Life Cycle Mode=::NewEnumerator1
    Loop Behavior=::NewEnumerator1
    Loop Duration Mode=::NewEnumerator1
    UseLoopDelay=false
   )
   &SpawnBurst_Instantaneous(Use Spawn Probability=false, UseLoopCountLimit=false)
   /SetVariables(Emitter.RainbowLine_Flat:NiagaraFloat=&MakeCustomFloatFromBool(Bool=&SetBoolByFloatComparison(A=$User.IsFrom3D, B=0), True Float=0, False Float=-1)) {isSetParameters=true, module=SetVariables_#9}
   /SetVariables(Emitter.RainbowLine_Strength:NiagaraFloat=&MakeCustomFloatFromBool(
     Bool=&SetBoolByFloatComparison(A=$User.IsFrom3D, B=0)
     True Float=4
     False Float=&FNTimeOfDayScalar(Night Scale=1.5, Morning Scale=0.4, Evening Scale=0.3)
    )) {isSetParameters=true, module=SetVariables_#10}
   /SetVariables(Emitter.RainbowLine_Power=0.5) {isSetParameters=true, module=SetVariables_#11}
   /SetVariables(Emitter.RainbowLine_Density=5) {isSetParameters=true, module=SetVariables_#12}
   /SetVariables(Emitter.RainbowLine_FlowSpeed=0.15) {isSetParameters=true, module=SetVariables_#13}
  ]
  particleSpawnScript [
   &InitializeParticle(
    Mesh Scale Mode=::NewEnumerator3
    Mesh Uniform Scale=0.5
    Position Offset=&Add_Vector(A=<0 1 0>, B=Linked() {linkedVariable {name=User.Sticker03 Position, type=NiagaraPosition}})
    Sprite Size Mode=::NewEnumerator3
    Uniform Sprite Size=70
    UsePositionOffset=true
    Write Material Random=true
   )
   &DynamicMaterialParameters(
    Index 0 Param 1=0.5
    Module.Param0WriteEnabled=true
    Module.Param1WriteEnabled=false
    Module.Param2WriteEnabled=false
    Module.Param3WriteEnabled=false
   )
   /SetVariables(Particles.RainbowLineSize:Vector2f=&Multiply_Vector2D_ByFloat(Float=0.5, Vector2D=$Particles.SpriteSize)) {isSetParameters=true, module=SetVariables_#14}
   &InitialMeshOrientation(Mesh Orientation Mode=::NewEnumerator2, Rotation=<0.04 0 0>, Use Rotation Vector=true)
  ]
  particleUpdateScript [
   &ParticleState(Kill Particles When Lifetime Has Elapsed=false, Let Infinitely Lived Particles Die When Emitter Deactivates=true)
   &ScaleMeshSize(Scale Factor=&Multiply_VectorByFloat(Vector=<1 1 1>, Float=$User.Sticker03 Size))
  ]
  rendererClasses [&NiagaraMeshRendererProperties]
  renderers [{rendererClass=&NiagaraMeshRendererProperties, rendererIndex=0}]
 }
 Sticker04 {
  emitterUpdateScript [
   &EmitterState(
    Inactive Response=::NewEnumerator1
    Life Cycle Mode=::NewEnumerator1
    Loop Behavior=::NewEnumerator1
    Loop Duration Mode=::NewEnumerator1
    UseLoopDelay=false
   )
   &SpawnBurst_Instantaneous(Use Spawn Probability=false, UseLoopCountLimit=false)
  ]
  particleSpawnScript [
   &InitializeParticle(
    Color=$System.Base color
    Mesh Scale=&Multiply_VectorByFloat(Vector=<1 1 1>, Float=0.5)
    Mesh Scale Mode=::NewEnumerator4
    Position Offset=&Add_Vector(A=Linked() {linkedVariable {name=User.Sticker04 Position, type=NiagaraPosition}}, B=<0 0 0>)
    UsePositionOffset=true
    Write Material Random=true
   )
   &DynamicMaterialParameters(
    Index 0 Param 2=&MakeFloatFromVector(Vector=$Particles.Scale, Channel=::NewEnumerator0)
    Module.Param0WriteEnabled=true
    Module.Param1WriteEnabled=true
    Module.Param2WriteEnabled=false
    Module.Param3WriteEnabled=false
   )
   /SetVariables(Particles.Mesh02 size:Vector3f=&Multiply_VectorByFloat(Vector=<1 1 1>, Float=$User.Sticker04 Size)) {isSetParameters=true, module=SetVariables_#6}
   /SetVariables(Particles.Mesh02 Color:LinearColor=<rgba 1 1 1 3>) {isSetParameters=true, module=SetVariables_#7}
   &InitialMeshOrientation(Mesh Orientation Mode=::NewEnumerator2, Rotation=<-0.02 0 0>, Use Rotation Vector=true)
   /SetVariables(Particles.Mesh02 Position:NiagaraPosition=&Add_Vector(A=Linked() {linkedVariable {name=Particles.Position, type=NiagaraPosition}}, B=<0 -1 0>)) {isSetParameters=true, module=SetVariables_#15}
  ]
  particleUpdateScript [
   &ParticleState(Kill Particles When Lifetime Has Elapsed=false, Let Infinitely Lived Particles Die When Emitter Deactivates=true)
   &ScaleMeshSize(Scale Factor=&Multiply_VectorByFloat(Vector=<1 1 1>, Float=$User.Sticker04 Size))
   &ScaleColor(
    Scale RGB=&Multiply_VectorByFloat(Vector=&Multiply_VectorByFloat(Vector=<1 1 1>, Float=$System.Bright), Float=$System.TOD)
    ScaleA=true
    ScaleRGB=true
   )
  ]
  rendererClasses [&NiagaraMeshRendererProperties]
  renderers [
   {rendererClass=&NiagaraMeshRendererProperties, rendererIndex=0}
   {rendererClass=&NiagaraMeshRendererProperties, rendererIndex=1}
  ]
 }
 Sticker04_PresetFX_RainbowLine001 {
  emitterUpdateScript [
   &EmitterState(
    Inactive Response=::NewEnumerator1
    Life Cycle Mode=::NewEnumerator1
    Loop Behavior=::NewEnumerator1
    Loop Duration Mode=::NewEnumerator1
    UseLoopDelay=false
   )
   &SpawnBurst_Instantaneous(Use Spawn Probability=false, UseLoopCountLimit=false)
   /SetVariables(Emitter.RainbowLine_Flat:NiagaraFloat=&MakeCustomFloatFromBool(Bool=&SetBoolByFloatComparison(A=$User.IsFrom3D, B=0), True Float=0, False Float=-1)) {isSetParameters=true, module=SetVariables_#9}
   /SetVariables(Emitter.RainbowLine_Strength:NiagaraFloat=&MakeCustomFloatFromBool(
     Bool=&SetBoolByFloatComparison(A=$User.IsFrom3D, B=0)
     True Float=4
     False Float=&FNTimeOfDayScalar(Night Scale=1.5, Morning Scale=0.4, Evening Scale=0.3)
    )) {isSetParameters=true, module=SetVariables_#10}
   /SetVariables(Emitter.RainbowLine_Power=0.5) {isSetParameters=true, module=SetVariables_#11}
   /SetVariables(Emitter.RainbowLine_Density=5) {isSetParameters=true, module=SetVariables_#12}
   /SetVariables(Emitter.RainbowLine_FlowSpeed=0.15) {isSetParameters=true, module=SetVariables_#13}
  ]
  particleSpawnScript [
   &InitializeParticle(
    Mesh Scale Mode=::NewEnumerator3
    Mesh Uniform Scale=0.5
    Position Offset=&Add_Vector(A=<0 -1 0>, B=Linked() {linkedVariable {name=User.Sticker04 Position, type=NiagaraPosition}})
    Sprite Size Mode=::NewEnumerator3
    Uniform Sprite Size=70
    UsePositionOffset=true
    Write Material Random=true
   )
   &DynamicMaterialParameters(Module.Param0WriteEnabled=true, Module.Param1WriteEnabled=false, Module.Param2WriteEnabled=false, Module.Param3WriteEnabled=false)
   /SetVariables(Particles.RainbowLineSize:Vector2f=&Multiply_Vector2D_ByFloat(Float=0.5, Vector2D=$Particles.SpriteSize)) {isSetParameters=true, module=SetVariables_#14}
   &InitialMeshOrientation(Mesh Orientation Mode=::NewEnumerator2, Rotation=<-0.02 0 0>, Use Rotation Vector=true)
  ]
  particleUpdateScript [
   &ParticleState(Kill Particles When Lifetime Has Elapsed=false, Let Infinitely Lived Particles Die When Emitter Deactivates=true)
   &ScaleMeshSize(Scale Factor=&Multiply_VectorByFloat(Vector=<1 1 1>, Float=$User.Sticker04 Size))
  ]
  rendererClasses [&NiagaraMeshRendererProperties]
  renderers [{rendererClass=&NiagaraMeshRendererProperties, rendererIndex=0}]
 }
 Sticker05 {
  emitterUpdateScript [
   &EmitterState(
    Inactive Response=::NewEnumerator1
    Life Cycle Mode=::NewEnumerator1
    Loop Behavior=::NewEnumerator1
    Loop Duration Mode=::NewEnumerator1
    UseLoopDelay=false
   )
   &SpawnBurst_Instantaneous(Use Spawn Probability=false, UseLoopCountLimit=false)
  ]
  particleSpawnScript [
   &InitializeParticle(
    Color=$System.Base color
    Mesh Scale=&Multiply_VectorByFloat(Vector=<1 1 1>, Float=0.5)
    Mesh Scale Mode=::NewEnumerator4
    Position Offset=&Add_Vector(A=Linked() {linkedVariable {name=User.Sticker05 Position, type=NiagaraPosition}}, B=<0 0 0>)
    UsePositionOffset=true
    Write Material Random=true
   )
   &DynamicMaterialParameters(
    Index 0 Param 1=0.9
    Index 0 Param 2=&MakeFloatFromVector(Vector=$Particles.Scale, Channel=::NewEnumerator0)
    Module.Param0WriteEnabled=true
    Module.Param1WriteEnabled=true
    Module.Param2WriteEnabled=false
    Module.Param3WriteEnabled=false
   )
   /SetVariables(Particles.Mesh02 size:Vector3f=&Multiply_VectorByFloat(Vector=<1 1 1>, Float=$User.Sticker05 Size)) {isSetParameters=true, module=SetVariables_#6}
   /SetVariables(Particles.Mesh02 Color:LinearColor=<rgba 1 1 1 3>) {isSetParameters=true, module=SetVariables_#7}
   &InitialMeshOrientation(Mesh Orientation Mode=::NewEnumerator2, Rotation=<-0.03 0 0>, Use Rotation Vector=true)
   &InitialMeshOrientation(Mesh Orientation Mode=::NewEnumerator2, Rotation=<0 0.01 0.07>, Use Rotation Vector=true) {module=InitialMeshOrientation001}
   /SetVariables(Particles.Mesh02 Position:NiagaraPosition=&Add_Vector(A=Linked() {linkedVariable {name=Particles.Position, type=NiagaraPosition}}, B=<0.3 -1 0>)) {isSetParameters=true, module=SetVariables_#15}
  ]
  particleUpdateScript [
   &ParticleState(Kill Particles When Lifetime Has Elapsed=false, Let Infinitely Lived Particles Die When Emitter Deactivates=true)
   &ScaleMeshSize(Scale Factor=&Multiply_VectorByFloat(Vector=<1 1 1>, Float=$User.Sticker05 Size))
   &ScaleColor(
    Scale RGB=&Multiply_VectorByFloat(Vector=&Multiply_VectorByFloat(Vector=<1 1 1>, Float=$System.Bright), Float=$System.TOD)
    ScaleA=true
    ScaleRGB=true
   )
  ]
  rendererClasses [&NiagaraMeshRendererProperties]
  renderers [
   {rendererClass=&NiagaraMeshRendererProperties, rendererIndex=0}
   {rendererClass=&NiagaraMeshRendererProperties, rendererIndex=1}
  ]
 }
 Sticker05_PresetFX_RainbowLine003 {
  emitterUpdateScript [
   &EmitterState(
    Inactive Response=::NewEnumerator1
    Life Cycle Mode=::NewEnumerator1
    Loop Behavior=::NewEnumerator1
    Loop Duration Mode=::NewEnumerator1
    UseLoopDelay=false
   )
   &SpawnBurst_Instantaneous(Use Spawn Probability=false, UseLoopCountLimit=false)
   /SetVariables(Emitter.RainbowLine_Flat:NiagaraFloat=&MakeCustomFloatFromBool(Bool=&SetBoolByFloatComparison(A=$User.IsFrom3D, B=0), True Float=0, False Float=-1)) {isSetParameters=true, module=SetVariables_#9}
   /SetVariables(Emitter.RainbowLine_Strength:NiagaraFloat=&MakeCustomFloatFromBool(
     Bool=&SetBoolByFloatComparison(A=$User.IsFrom3D, B=0)
     True Float=4
     False Float=&FNTimeOfDayScalar(Night Scale=1.5, Morning Scale=0.4, Evening Scale=0.3)
    )) {isSetParameters=true, module=SetVariables_#10}
   /SetVariables(Emitter.RainbowLine_Power=0.5) {isSetParameters=true, module=SetVariables_#11}
   /SetVariables(Emitter.RainbowLine_Density=5) {isSetParameters=true, module=SetVariables_#12}
   /SetVariables(Emitter.RainbowLine_FlowSpeed=0.15) {isSetParameters=true, module=SetVariables_#13}
  ]
  particleSpawnScript [
   &InitializeParticle(
    Mesh Scale Mode=::NewEnumerator3
    Mesh Uniform Scale=0.5
    Position Offset=&Add_Vector(A=<0.5 -1 0>, B=Linked() {linkedVariable {name=User.Sticker05 Position, type=NiagaraPosition}})
    Sprite Size Mode=::NewEnumerator3
    Uniform Sprite Size=70
    UsePositionOffset=true
    Write Material Random=true
   )
   &DynamicMaterialParameters(
    Index 0 Param 1=0.3
    Module.Param0WriteEnabled=true
    Module.Param1WriteEnabled=false
    Module.Param2WriteEnabled=false
    Module.Param3WriteEnabled=false
   )
   /SetVariables(Particles.RainbowLineSize:Vector2f=&Multiply_Vector2D_ByFloat(Float=0.5, Vector2D=$Particles.SpriteSize)) {isSetParameters=true, module=SetVariables_#14}
   &InitialMeshOrientation(Mesh Orientation Mode=::NewEnumerator2, Rotation=<-0.03 0 0>, Use Rotation Vector=true)
   &InitialMeshOrientation(Mesh Orientation Mode=::NewEnumerator2, Rotation=<0 0.03 0.07>, Use Rotation Vector=true) {module=InitialMeshOrientation001}
  ]
  particleUpdateScript [
   &ParticleState(Kill Particles When Lifetime Has Elapsed=false, Let Infinitely Lived Particles Die When Emitter Deactivates=true)
   &ScaleMeshSize(Scale Factor=&Multiply_VectorByFloat(Vector=<1 1 1>, Float=$User.Sticker05 Size))
  ]
  rendererClasses [&NiagaraMeshRendererProperties]
  renderers [{rendererClass=&NiagaraMeshRendererProperties, rendererIndex=0}]
 }
 Sticker06 {
  emitterUpdateScript [
   &EmitterState(
    Inactive Response=::NewEnumerator1
    Life Cycle Mode=::NewEnumerator1
    Loop Behavior=::NewEnumerator1
    Loop Duration Mode=::NewEnumerator1
    UseLoopDelay=false
   )
   &SpawnBurst_Instantaneous(Use Spawn Probability=false, UseLoopCountLimit=false)
  ]
  particleSpawnScript [
   &InitializeParticle(
    Color=$System.Base color
    Mesh Scale=&Multiply_VectorByFloat(Vector=<1 1 1>, Float=0.5)
    Mesh Scale Mode=::NewEnumerator4
    Position Offset=&Add_Vector(A=Linked() {linkedVariable {name=User.Sticker06 Position, type=NiagaraPosition}}, B=<0 0 0>)
    UsePositionOffset=true
    Write Material Random=true
   )
   &DynamicMaterialParameters(
    Index 0 Param 1=0.8
    Index 0 Param 2=&MakeFloatFromVector(Vector=$Particles.Scale, Channel=::NewEnumerator0)
    Module.Param0WriteEnabled=true
    Module.Param1WriteEnabled=true
    Module.Param2WriteEnabled=false
    Module.Param3WriteEnabled=false
   )
   /SetVariables(Particles.Mesh02 size:Vector3f=&Multiply_VectorByFloat(Vector=<1 1 1>, Float=$User.Sticker05 Size)) {isSetParameters=true, module=SetVariables_#6}
   /SetVariables(Particles.Mesh02 Color:LinearColor=<rgba 1 1 1 3>) {isSetParameters=true, module=SetVariables_#7}
   &InitialMeshOrientation(Mesh Orientation Mode=::NewEnumerator2, Rotation=<-0.02 0 0>, Use Rotation Vector=true)
   &InitialMeshOrientation(Mesh Orientation Mode=::NewEnumerator2, Rotation=<0 0 -0.07>, Use Rotation Vector=true) {module=InitialMeshOrientation001}
   /SetVariables(Particles.Mesh02 Position:NiagaraPosition=&Add_Vector(A=Linked() {linkedVariable {name=Particles.Position, type=NiagaraPosition}}, B=<-0.3 -1 0>)) {isSetParameters=true, module=SetVariables_#15}
  ]
  particleUpdateScript [
   &ParticleState(Kill Particles When Lifetime Has Elapsed=false, Let Infinitely Lived Particles Die When Emitter Deactivates=true)
   &ScaleMeshSize(Scale Factor=&Multiply_VectorByFloat(Vector=<1 1 1>, Float=$User.Sticker05 Size))
   &ScaleColor(
    Scale RGB=&Multiply_VectorByFloat(Vector=&Multiply_VectorByFloat(Vector=<1 1 1>, Float=$System.Bright), Float=$System.TOD)
    ScaleA=true
    ScaleRGB=true
   )
  ]
  rendererClasses [&NiagaraMeshRendererProperties]
  renderers [
   {rendererClass=&NiagaraMeshRendererProperties, rendererIndex=0}
   {rendererClass=&NiagaraMeshRendererProperties, rendererIndex=1}
  ]
 }
 Sticker06_PresetFX_RainbowLine004 {
  emitterUpdateScript [
   &EmitterState(
    Inactive Response=::NewEnumerator1
    Life Cycle Mode=::NewEnumerator1
    Loop Behavior=::NewEnumerator1
    Loop Duration Mode=::NewEnumerator1
    UseLoopDelay=false
   )
   &SpawnBurst_Instantaneous(Use Spawn Probability=false, UseLoopCountLimit=false)
   /SetVariables(Emitter.RainbowLine_Flat:NiagaraFloat=&MakeCustomFloatFromBool(Bool=&SetBoolByFloatComparison(A=$User.IsFrom3D, B=0), True Float=0, False Float=-1)) {isSetParameters=true, module=SetVariables_#9}
   /SetVariables(Emitter.RainbowLine_Strength:NiagaraFloat=&MakeCustomFloatFromBool(
     Bool=&SetBoolByFloatComparison(A=$User.IsFrom3D, B=0)
     True Float=4
     False Float=&FNTimeOfDayScalar(Night Scale=1.5, Morning Scale=0.4, Evening Scale=0.3)
    )) {isSetParameters=true, module=SetVariables_#10}
   /SetVariables(Emitter.RainbowLine_Power=0.5) {isSetParameters=true, module=SetVariables_#11}
   /SetVariables(Emitter.RainbowLine_Density=5) {isSetParameters=true, module=SetVariables_#12}
   /SetVariables(Emitter.RainbowLine_FlowSpeed=0.15) {isSetParameters=true, module=SetVariables_#13}
  ]
  particleSpawnScript [
   &InitializeParticle(
    Mesh Scale Mode=::NewEnumerator3
    Mesh Uniform Scale=0.5
    Position Offset=&Add_Vector(A=<-0.5 -1 0>, B=Linked() {linkedVariable {name=User.Sticker06 Position, type=NiagaraPosition}})
    Sprite Size Mode=::NewEnumerator3
    Uniform Sprite Size=70
    UsePositionOffset=true
    Write Material Random=true
   )
   &DynamicMaterialParameters(
    Index 0 Param 1=0.9
    Module.Param0WriteEnabled=true
    Module.Param1WriteEnabled=false
    Module.Param2WriteEnabled=false
    Module.Param3WriteEnabled=false
   )
   /SetVariables(Particles.RainbowLineSize:Vector2f=&Multiply_Vector2D_ByFloat(Float=0.5, Vector2D=$Particles.SpriteSize)) {isSetParameters=true, module=SetVariables_#14}
   &InitialMeshOrientation(Mesh Orientation Mode=::NewEnumerator2, Rotation=<-0.02 0 0>, Use Rotation Vector=true)
   &InitialMeshOrientation(Mesh Orientation Mode=::NewEnumerator2, Rotation=<0 0.01 -0.07>, Use Rotation Vector=true) {module=InitialMeshOrientation001}
  ]
  particleUpdateScript [
   &ParticleState(Kill Particles When Lifetime Has Elapsed=false, Let Infinitely Lived Particles Die When Emitter Deactivates=true)
   &ScaleMeshSize(Scale Factor=&Multiply_VectorByFloat(Vector=<1 1 1>, Float=$User.Sticker05 Size))
  ]
  rendererClasses [&NiagaraMeshRendererProperties]
  renderers [{rendererClass=&NiagaraMeshRendererProperties, rendererIndex=0}]
 }
}
summary {
 systemName=NS_Backpack_Candor_Tech_Pink
 emitters [
  {emitterName=Sticker01, rendererClasses [&NiagaraMeshRendererProperties]}
  {emitterName=Sticker01_PresetFX_RainbowLine, rendererClasses [&NiagaraMeshRendererProperties]}
  {emitterName=Sticker04, rendererClasses [&NiagaraMeshRendererProperties]}
  {emitterName=Sticker04_PresetFX_RainbowLine001, rendererClasses [&NiagaraMeshRendererProperties]}
  {emitterName=Sticker03, rendererClasses [&NiagaraMeshRendererProperties]}
  {emitterName=Sticker03_PresetFX_RainbowLine002, rendererClasses [&NiagaraMeshRendererProperties]}
  {emitterName=Sticker05, rendererClasses [&NiagaraMeshRendererProperties]}
  {emitterName=Sticker05_PresetFX_RainbowLine003, rendererClasses [&NiagaraMeshRendererProperties]}
  {emitterName=Sticker06, rendererClasses [&NiagaraMeshRendererProperties]}
  {emitterName=Sticker06_PresetFX_RainbowLine004, rendererClasses [&NiagaraMeshRendererProperties]}
  {bEnabled=false, emitterName=Glow, rendererClasses [&NiagaraSpriteRendererProperties]}
 ]
 userVariables [
  User.Backpack_DiffuseTex {
   type=Texture
   description=""
   defaultValue {
    struct=NiagaraExt_VariableValue_Object
    value {
     object=/BRCosmetics/Accessories/FORT_Backpacks/Backpack_M_MED_Candor_Tech_Pink/FX/Texture/T_Candor_tech_pink_backpack_Sticker01
     objectClass=/Script/Engine.Texture
    }
   }
  }
  User.Backpack_DiffuseTex03 {
   type=Texture
   description=""
   defaultValue {
    struct=NiagaraExt_VariableValue_Object
    value {
     object=/BRCosmetics/Accessories/FORT_Backpacks/Backpack_M_MED_Candor_Tech_Pink/FX/Texture/T_Candor_tech_pink_backpack_Sticker03
     objectClass=&Texture
    }
   }
  }
  User.Backpack_DiffuseTex04 {
   type=Texture
   description=""
   defaultValue {
    struct=NiagaraExt_VariableValue_Object
    value {
     object=/BRCosmetics/Accessories/FORT_Backpacks/Backpack_M_MED_Candor_Tech_Pink/FX/Texture/T_Candor_tech_pink_backpack_Sticker04
     objectClass=&Texture
    }
   }
  }
  User.Backpack_DiffuseTex05 {
   type=Texture
   description=""
   defaultValue {
    struct=NiagaraExt_VariableValue_Object
    value {
     object=/BRCosmetics/Accessories/FORT_Backpacks/Backpack_M_MED_Candor_Tech_Pink/FX/Texture/T_Candor_tech_pink_backpack_Sticker05
     objectClass=&Texture
    }
   }
  }
  User.Backpack_DiffuseTex06 {
   type=Texture
   description=""
   defaultValue {
    struct=NiagaraExt_VariableValue_Object
    value {
     object=/BRCosmetics/Accessories/FORT_Backpacks/Backpack_M_MED_Candor_Tech_Pink/FX/Texture/T_Candor_tech_pink_backpack_Sticker06
     objectClass=&Texture
    }
   }
  }
  User.Backpack_Mat {
   type=MaterialInterface
   description=""
   defaultValue {
    struct=NiagaraExt_VariableValue_Object
    value {
     object=/BRCosmetics/Accessories/FORT_Backpacks/Backpack_M_MED_Candor_Tech_Pink/FX/Material/MI_Candor_tech_pink_backpack_Sticker01
     objectClass=/Script/Engine.MaterialInterface
    }
   }
  }
  User.Backpack_Mat03 {
   type=MaterialInterface
   description=""
   defaultValue {
    struct=NiagaraExt_VariableValue_Object
    value {
     object=/BRCosmetics/Accessories/FORT_Backpacks/Backpack_M_MED_Candor_Tech_Pink/FX/Material/MI_Candor_tech_pink_backpack_Sticker03
     objectClass=&MaterialInterface
    }
   }
  }
  User.Backpack_Mat04 {
   type=MaterialInterface
   description=""
   defaultValue {
    struct=NiagaraExt_VariableValue_Object
    value {
     object=/BRCosmetics/Accessories/FORT_Backpacks/Backpack_M_MED_Candor_Tech_Pink/FX/Material/MI_Candor_tech_pink_backpack_Sticker04
     objectClass=&MaterialInterface
    }
   }
  }
  User.Backpack_Mat05 {
   type=MaterialInterface
   description=""
   defaultValue {
    struct=NiagaraExt_VariableValue_Object
    value {
     object=/BRCosmetics/Accessories/FORT_Backpacks/Backpack_M_MED_Candor_Tech_Pink/FX/Material/MI_Candor_tech_pink_backpack_Sticker05
     objectClass=&MaterialInterface
    }
   }
  }
  User.Backpack_Mat06 {
   type=MaterialInterface
   description=""
   defaultValue {
    struct=NiagaraExt_VariableValue_Object
    value {
     object=/BRCosmetics/Accessories/FORT_Backpacks/Backpack_M_MED_Candor_Tech_Pink/FX/Material/MI_Candor_tech_pink_backpack_Sticker06
     objectClass=&MaterialInterface
    }
   }
  }
  User.Backpack_NormalTex {
   type=Texture
   description=""
   defaultValue {
    struct=NiagaraExt_VariableValue_Object
    value {
     object=/BRCosmetics/Accessories/FORT_Backpacks/Backpack_M_MED_Candor_Tech_Pink/FX/Texture/T_Candor_tech_pink_backpack_normal_Sticker01
     objectClass=&Texture
    }
   }
  }
  User.Backpack_NormalTex03 {
   type=Texture
   description=""
   defaultValue {
    struct=NiagaraExt_VariableValue_Object
    value {
     object=/BRCosmetics/Accessories/FORT_Backpacks/Backpack_M_MED_Candor_Tech_Pink/FX/Texture/T_Candor_tech_pink_backpack_normal_Sticker03
     objectClass=&Texture
    }
   }
  }
  User.Backpack_NormalTex04 {
   type=Texture
   description=""
   defaultValue {
    struct=NiagaraExt_VariableValue_Object
    value {
     object=/BRCosmetics/Accessories/FORT_Backpacks/Backpack_M_MED_Candor_Tech_Pink/FX/Texture/T_Candor_tech_pink_backpack_normal_Sticker04
     objectClass=&Texture
    }
   }
  }
  User.Backpack_NormalTex05 {
   type=Texture
   description=""
   defaultValue {
    struct=NiagaraExt_VariableValue_Object
    value {
     object=/BRCosmetics/Accessories/FORT_Backpacks/Backpack_M_MED_Candor_Tech_Pink/FX/Texture/T_Candor_tech_pink_backpack_normal_Sticker05
     objectClass=&Texture
    }
   }
  }
  User.Backpack_NormalTex06 {
   type=Texture
   description=""
   defaultValue {
    struct=NiagaraExt_VariableValue_Object
    value {
     object=/BRCosmetics/Accessories/FORT_Backpacks/Backpack_M_MED_Candor_Tech_Pink/FX/Texture/T_Candor_tech_pink_backpack_normal_Sticker06
     objectClass=&Texture
    }
   }
  }
  User.bIsFrontEnd {type=NiagaraBool}
  User.bIsFrontEndPreview {type=NiagaraBool}
  User.IsFrom3D {type=NiagaraFloat, defaultValue=0, description=""}
  User.Sticker01 Position {type=NiagaraPosition, description="", defaultValue {}}
  User.Sticker01 Size {type=NiagaraFloat, defaultValue=0.7, description=""}
  User.Sticker03 Position {type=NiagaraPosition, description="", defaultValue {}}
  User.Sticker03 Size {type=NiagaraFloat, defaultValue=0.55, description=""}
  User.Sticker04 Position {type=NiagaraPosition, description="", defaultValue {}}
  User.Sticker04 Size {type=NiagaraFloat, defaultValue=0.45, description=""}
  User.Sticker05 Position {type=NiagaraPosition, description="", defaultValue {}}
  User.Sticker05 Size {type=NiagaraFloat, defaultValue=0.4, description=""}
  User.Sticker06 Position {type=NiagaraPosition, description="", defaultValue {}}
  User.Sticker06 Size {type=NiagaraFloat, defaultValue=0.4, description=""}
 ]
}
